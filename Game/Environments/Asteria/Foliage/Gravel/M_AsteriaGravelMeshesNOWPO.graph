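# Nanite and instancing-ready material for Asteria gravel meshes, with no world position offset.
# Vertex-color RGB picks per-biome tints, blended by terrain masks and optional biome colors, with switches for texture maps and snow on top.
# A darker-albedo, lower-roughness wet variant blends in by mask, and detail normals become world space.
# An optional RVT switch fades mesh bases into the landscape by height, while forward and mobile paths bypass it.
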Material /Game/Environments/Asteria/Foliage/Gravel/M_AsteriaGravelMeshesNOWPO
graph {
 properties {
  bTangentSpaceNormal=false
  bUseMaterialAttributes=true
  bUsedWithInstancedStaticMeshes=true
  bUsedWithNanite=true
  pixelDepthOffsetMode=PDOM_Legacy
  usedShadingModels=Default Lit
 }
 nodes [
  Constant() {r=1}
  MaterialFunctionCall(Texture=NamedRerouteUsage()) {materialFunction=&HueShift}
  n1 = MaterialFunctionCall!(UseLandscapeCoords) {materialFunction=/Game/Athena/Asteria/Environments/Landscape/Materials/MaterialFunctions/MF_AsteriaTerrainMasks}
  NamedRerouteDeclaration(None=@n1.Tundra) {name=TundraBiome, nodeColor=<rgba 0.875 1 0 1>}
  n2 = MaterialFunctionCall!(
   Diffuse Texture Object=TextureObject() {samplerType=SAMPLERTYPE_Masks, texture=/Game/Athena/WinterArt/Textures/T_DirtLeaks_M}
   Spec Texture Object
   TextureSize=ScalarParameter() {defaultValue=64, parameterName=DetailNoiseTilinjg}
   --------------------------
   Use seperate settings for Z projection
   Use seperate texture for Z projection
   Texture Size Z projection
   Z Plane Diffuse Texture Object
   Z Plane Spec Texture Object
   "--------------------------      "
   ProjectionTransitionContrast
   Use Texture Alpha Heightmap to Refine the Transistion
   HeightLerpTransitionContrast
   PickTheDiffuseHeightChannel
   TheDiffuseHeightChannel
   "--------------------------       "
   NormalTextureObject=TextureObject() {samplerType=SAMPLERTYPE_Normal, texture=/Game/Environments/Apollo/Rocks/Natural_Rocks/Textures/T_RockDetails_02_N}
   World Space Normals=StaticBool()
   "--------------------------  "
   Export Float 4 Diffuse
   Export Float 4 Spec
   "--------------------------     "
   WorldPosition
   World Space Normal
  ) {materialFunction=/Engine/Functions/Engine_MaterialFunctions02/Texturing/WorldAlignedTextures_Complex}
  n3 = MaterialFunctionCall(Float3=@n2.Z Diffuse Texture) {materialFunction=&BreakOutFloat3Components}
  NamedRerouteDeclaration(None=@n3.B) {name=Details-B, nodeColor=<rgba 0.4062 1 0 1>}
  n4 = TextureSampleParameter2D() {parameterName=T_Specular, samplerType=SAMPLERTYPE_Masks, texture=/Engine/EngineMaterials/DefaultDiffuse_TC_Masks}
  NamedRerouteDeclaration(None=@n4.R) {name=Texture_Specular, nodeColor=<rgba 0 0.8359 1 1>}
  NamedRerouteDeclaration(None=@n4.B) {name=Texture_Roughness, nodeColor=<rgba 1 0.375 0 1>}
  NamedRerouteDeclaration(None=TextureSampleParameter2D() {parameterName=T_Normal, samplerType=SAMPLERTYPE_Normal, texture=/Engine/EngineMaterials/FlatNormal}.RGB) {name=Texture_Normal, nodeColor=<rgba 1 0.7266 0 1>}
  NamedRerouteDeclaration(None=Desaturation(None=NamedRerouteUsage(), Fraction=ScalarParameter() {defaultValue=1, parameterName=TextureDesat})) {name=Diffuse_Greyscale, nodeColor=<rgba 0 1 0.0625 1>}
  n5 = VertexNormalWS()
  n6 = ConstantBiasScale(None=DotProduct(A=Constant3Vector() {constant=<rgba 0 0 1 0>}, B=@n5))
  n7 = ScalarParameter() {defaultValue=1, parameterName=G_SnowSample}
  n8 = ScalarParameter() {defaultValue=0.3, parameterName=G_SnowSampleRadius}
  n9 = ScalarParameter() {defaultValue=0.5, parameterName=G_SnowSampleHardness}
  NamedRerouteDeclaration(None=Multiply(A=SphereMask(A=@n6, B=@n7, Radius=@n8, Hardness=@n9), B=NamedRerouteUsage())) {name=G_SnowTop_Mask, nodeColor=<rgba 0.1953 0 1 1>}
  n10 = Add(
   A=ComponentMask(None=WorldPosition() {worldPositionShaderOffset=WPT_ExcludeAllShaderOffsets}.XYZ) {b=true}
   B=NamedRerouteDeclaration(None=ScalarParameter() {defaultValue=64, parameterName=WaterLevelOffset}) {name=WaterLevelOffset, nodeColor=<rgba 1 0 1 1>}
  )
  n11 = Constant() {desc=slightly wet sand height offset, r=10}
  n12 = Saturate(None=Divide(A=Add(A=@n10, B=@n11), B=ScalarParameter() {defaultValue=90, parameterName=WaterLevelTransition}))
  NamedRerouteDeclaration(None=@n12) {name=WaterLevel, nodeColor=<rgba 1 0 1 1>}
  NamedRerouteDeclaration(None=@n12) {name=OceanMask, nodeColor=<rgba 0.5 0 1 1>}
  NamedRerouteDeclaration(None=Multiply(A=@n2.Normal Z Texture, B=VectorParameter() {parameterName=DetailNormalStrength, defaultValue=<rgba 2 2 0 1>}.RGB)) {name=Details N, nodeColor=<rgba 0 1 0.5078 1>}
  n13 = TextureObject() {samplerType=SAMPLERTYPE_Grayscale, texture=/Game/Weapons/FORT_Melee/Pickaxe_Cube_Ninja_Male/FX/T_CellNoise}
  n14 = Multiply(A=ScalarParameter() {defaultValue=0.0002, parameterName=GravelVariationMask_02_Tiling}, B=WorldPosition().XYZ)
  n15 = Multiply(
   A=ComponentMask(None=MaterialFunctionCall!(
     Texture=@n13
     LandscapeCoordinates=@n14
     LowEndTiling
     HexGridGradientSharpness
     HexGridSize
     HexGridUVInput
     MinScale
     MaxScale
     MinOffset
     MaxOffset
     "MaxRotation(Degrees)"
     "MinRotation(Degrees)"
     forceLowQuality
    ) {materialFunction=/Game/Athena/Asteria/Environments/Landscape/Materials/MaterialFunctions/UtilityFunctions/Landscape_TextureSample}.Result) {r=true}
   B=ScalarParameter() {defaultValue=7, parameterName=GravelMask_v2_Strength}
  )
  n16 = ScalarParameter() {defaultValue=4, parameterName=GravelVariationMask_02_Pow}
  NamedRerouteDeclaration(None=Clamp(None=Power(Base=@n15, Exp=@n16))) {name=VariationMask_02, nodeColor=<rgba 1 0 0.6562 1>}
  n17 = WorldPosition()
  n18 = NamedRerouteDeclaration(None=Subtract(A=ComponentMask(None=@n17.XYZ) {b=true}, B=ScalarParameter() {defaultValue=-55, parameterName=Waterline Height})) {name=Waterline H, nodeColor=<rgba 1 0 0.3984 1>}
  n19 = Constant() {r=35}
  NamedRerouteDeclaration(None=Saturate(None=Divide(A=@n18, B=@n19))) {name=Waterline Soft, nodeColor=<rgba 0 0.5078 1 1>}
  NamedRerouteDeclaration(None=@n3.R) {name=Details-R, nodeColor=<rgba 0 1 0.5078 1>}
  n20 = VectorParameter() {group=Color Overlay, parameterName=BaseDiffuse Diffuse-common, defaultValue=<rgba 0.0434 0.04 0.0303 0>}
  n21 = VectorParameter() {group=Color Overlay, parameterName=R-Color-common, defaultValue=<rgba 0.0605 0.0747 0.0691 0>}
  n22 = VertexColor()
  n23 = LinearInterpolate(
   A=LinearInterpolate(
    A=LinearInterpolate(A=@n20.RGB, B=@n21.RGB, Alpha=@n22.R)
    B=VectorParameter() {group=Color Overlay, parameterName=G-Color-common, defaultValue=<rgba 0.1372 0.0931 0.0593 0>}.RGB
    Alpha=@n22.G
   )
   B=VectorParameter() {group=Color Overlay, parameterName=B-Color-common, defaultValue=<rgba 0.0677 0.0557 0.0469 0>}.RGB
   Alpha=@n22.B
  )
  n24 = Multiply(A=VectorParameter() {parameterName=Base-MacroHueShift, defaultValue=<rgba 0.0704 0.0592 0.0513 1>}.RGB, B=@n23)
  n25 = ScalarParameter() {defaultValue=10, parameterName=BaseMacroHueValue}
  n26 = VertexColor()
  n27 = LinearInterpolate(
   A=VectorParameter() {group=Color Overlay, parameterName=BaseDiffuse-Tundra, defaultValue=<rgba 0.0434 0.04 0.0303 0>}.RGB
   B=VectorParameter() {group=Color Overlay, parameterName=R-Color-Tundra, defaultValue=<rgba 0.0605 0.0747 0.0691 0>}.RGB
   Alpha=@n26.R
  )
  n28 = VectorParameter() {group=Color Overlay, parameterName=G-Color-Tundra, defaultValue=<rgba 0.1372 0.0931 0.0593 0>}
  n29 = LinearInterpolate(
   A=LinearInterpolate(A=Multiply(A=@n24, B=@n25), B=@n23, Alpha=NamedRerouteUsage())
   B=LinearInterpolate(
    A=LinearInterpolate(A=@n27, B=@n28.RGB, Alpha=@n26.G)
    B=VectorParameter() {group=Color Overlay, parameterName=B-Color-Tundra, defaultValue=<rgba 0.0677 0.0557 0.0469 0>}.RGB
    Alpha=@n26.B
   )
   Alpha=NamedRerouteUsage()
  )
  n30 = Multiply(A=@n29, B=Clamp(None=NamedRerouteUsage(), Min=ScalarParameter() {defaultValue=0.2, parameterName=TextureminInfluence}))
  n31 = NamedRerouteUsage()
  n32 = NamedRerouteUsage()
  n33 = LinearInterpolate(A=@n30, B=LinearInterpolate(A=@n31, B=Multiply(A=@n31, B=@n30), Alpha=ScalarParameter() {parameterName=TempLerpJungleBiome}), Alpha=@n32)
  n34 = VectorParameter() {parameterName=SakuraBiomeBaseColorTint, defaultValue=<rgba 0.2066 0.1594 0.2396 1>}
  n35 = NamedRerouteUsage()
  n36 = VectorParameter() {parameterName=Fall-MacroHueShift, defaultValue=<rgba 0.148 0.1593 0.2 1>}
  n37 = NamedRerouteDeclaration(None=VectorParameter() {group=Color Overlay, parameterName=BaseDiffuse-Fall, defaultValue=<rgba 0.149 0.1061 0.1059 0>}.RGB) {name=BaseDiffuse-Fall, nodeColor=<rgba 0.3125 0 1 1>}
  n38 = VectorParameter() {group=Color Overlay, parameterName=R-Color-Fall, defaultValue=<rgba 0.1216 0.102 0.0706 0>}
  n39 = VertexColor()
  n40 = LinearInterpolate(
   A=LinearInterpolate(
    A=LinearInterpolate(A=@n37, B=@n38.RGB, Alpha=@n39.R)
    B=VectorParameter() {group=Color Overlay, parameterName=G-Color-Fall, defaultValue=<rgba 0.149 0.102 0.0667 0>}.RGB
    Alpha=@n39.G
   )
   B=VectorParameter() {group=Color Overlay, parameterName=B-Color-Fall, defaultValue=<rgba 0.0677 0.0557 0.0469 0>}.RGB
   Alpha=@n39.B
  )
  n41 = LinearInterpolate(
   A=Multiply(A=Multiply(A=@n36.RGB, B=@n40), B=ScalarParameter() {defaultValue=5, parameterName=FallMacroHueValue})
   B=@n40
   Alpha=NamedRerouteUsage()
  )
  n42 = NamedRerouteDeclaration(None=VectorParameter() {group=Color Overlay, parameterName=BaseDiffuse-Fall, defaultValue=<rgba 0.149 0.1061 0.1059 0>}.RGB) {name=BaseDiffuse-Fall_1, nodeColor=<rgba 1 0.3516 0 1>}
  n43 = VectorParameter() {group=Color Overlay, parameterName=R-Color-Fall, defaultValue=<rgba 0.1216 0.102 0.0706 0>}
  n44 = VertexColor()
  n45 = LinearInterpolate(
   A=LinearInterpolate(
    A=LinearInterpolate(A=@n42, B=@n43.RGB, Alpha=@n44.R)
    B=VectorParameter() {group=Color Overlay, parameterName=G-Color-Fall, defaultValue=<rgba 0.149 0.102 0.0667 0>}.RGB
    Alpha=@n44.G
   )
   B=VectorParameter() {group=Color Overlay, parameterName=B-Color-Fall, defaultValue=<rgba 0.0677 0.0557 0.0469 0>}.RGB
   Alpha=@n44.B
  )
  n46 = Multiply(A=Multiply(A=VectorParameter() {parameterName=Fall-MacroHueShift, defaultValue=<rgba 0.148 0.1593 0.2 1>}.RGB, B=@n45), B=Constant() {r=5})
  n47 = NamedRerouteUsage()
  n48 = StaticSwitchParameter(
   True=LinearInterpolate(A=@n33, B=Multiply(A=@n33, B=@n34.RGB), Alpha=@n35)
   False=LinearInterpolate(A=LinearInterpolate(A=@n29, B=@n41, Alpha=@n32), B=LinearInterpolate(A=@n46, B=@n45, Alpha=@n47), Alpha=@n35)
  ) {parameterName=UseDiffuseTexture}
  n49 = StaticSwitchParameter(
   True=MaterialFunctionCall!(Grass, Gravel=@n48, Mud, Sand, BiomeMask-RGBA, UseBiomeColorMask, UseLandscapeCoords) {materialFunction=/Game/Athena/Asteria/Environments/Landscape/Materials/MaterialFunctions/MF_AsteriaBiomeColors}.Gravel
   False=@n48
  ) {defaultValue=true, parameterName=UseBiomeColors}
  n50 = LinearInterpolate(
   A=@n49
   B=MaterialFunctionCall() {materialFunction=/Game/Athena/Asteria/Environments/Landscape/Materials/MaterialFunctions/MF_SnowColorBase}.SnowAlbedo
   Alpha=NamedRerouteUsage()
  )
  n51 = ScalarParameter() {defaultValue=0.25, parameterName=Scree_SpecMin}
  n52 = ScalarParameter() {defaultValue=1, parameterName=Scree_SpecMax}
  n53 = NamedRerouteUsage()
  n54 = ScalarParameter() {defaultValue=0.1, parameterName=Pebbles Roughness Min}
  n55 = ScalarParameter() {defaultValue=1, parameterName=Pebbles Roughness Max}
  n56 = NamedRerouteUsage()
  n57 = NamedRerouteUsage()
  n58 = Constant3Vector() {constant=<rgba 0 0 1 0>}
  n59 = SetMaterialAttributes!(
   MaterialAttributes
   Base Color=MaterialFunctionCall(Base=StaticSwitchParameter(True=@n50, False=@n49) {parameterName=TopWithSnow}, Blend=NamedRerouteUsage()) {materialFunction=&Blend_Overlay}.Result
   Specular=StaticSwitchParameter(True=NamedRerouteUsage(), False=LinearInterpolate(A=@n51, B=@n52, Alpha=@n53)) {parameterName=UseDiffuseTexture}
   Roughness=StaticSwitchParameter(True=NamedRerouteUsage(), False=LinearInterpolate(A=@n54, B=@n55, Alpha=@n56)) {parameterName=UseDiffuseTexture}
   Normal=Add(A=StaticSwitchParameter(True=@n57, False=@n58) {parameterName=UseDiffuseTexture}, B=NamedRerouteUsage())
  ) {attributeSetTypes [
    #1
    #2
    #3
    #4
   ]}
  n60 = GetMaterialAttributes(None=@n59) {attributeGetTypes [#1, #2, #3]}
  n61 = Multiply(A=@n60.Base Color, B=ScalarParameter() {defaultValue=1, parameterName=Wet_AlbedoDarken})
  n62 = Multiply(A=@n60.Roughness, B=ScalarParameter() {defaultValue=0.5, parameterName=Wet_RoughnessMult})
  n63 = GetMaterialAttributes(None=BlendMaterialAttributes(
    A=SetMaterialAttributes!(MaterialAttributes=@n60.MaterialAttributes, Base Color=@n61, Specular=@n60.MaterialAttributes, Roughness=@n62) {attributeSetTypes [#1, #2, #3]}
    B=@n59
    Alpha=NamedRerouteUsage()
   )) {attributeGetTypes [#4]}
  NamedRerouteDeclaration(None=SetMaterialAttributes!(MaterialAttributes=@n63.MaterialAttributes, Normal=Transform(None=@n63.Normal)) {attributeSetTypes [#4]}) {name=pre RVT, nodeColor=<rgba 1 0 0.2109 1>}
  NamedRerouteDeclaration(None=@n1.GrassLand) {name=FallBiome, nodeColor=<rgba 1 0.1172 0 1>}
  NamedRerouteDeclaration(None=@n1.Desert) {name=SakuraBiome, nodeColor=<rgba 0 0.9531 1 1>}
  NamedRerouteDeclaration(None=TextureSampleParameter2D() {parameterName=T_BaseColor}.RGB) {name=Texture_BaseColor, nodeColor=<rgba 0 1 0.1562 1>}
  NamedRerouteDeclaration(None=@n3.G) {name=Details-G, nodeColor=<rgba 0.6172 0 1 1>}
  Reroute()
  n64 = PixelNormalWS()
  ShaderStageSwitch(PixelShader=@n64, VertexShader=@n5)
  n65 = NamedRerouteUsage()
  n66 = RuntimeVirtualTextureSampleParameter() {materialType=BaseColor_Normal_Specular_YCoCg, parameterName=RVT, virtualTexture=/Game/Athena/Apollo/Maps/Landscape/VT/RVT_Apollo_Landscape}
  n67 = SetMaterialAttributes!(
   MaterialAttributes
   Base Color=@n66.BaseColor
   Specular=@n66.BaseColor
   Roughness=@n66.BaseColor
   Emissive Color=Constant()
   Normal=@n66.BaseColor
  ) {attributeSetTypes [
    #1
    #2
    #3
    #5
    #4
   ]}
  n68 = NamedRerouteUsage()
  n69 = GetMaterialAttributes(None=@n68) {attributeGetTypes [#4]}
  n70 = ScalarParameter() {defaultValue=1, group=RVT, parameterName=Blend RVT}
  n71 = LinearInterpolate(
   A=Multiply(A=ScalarParameter() {defaultValue=1, group=RVT, parameterName=RVT Scale Distance Blend}, B=@n70) {constA=3}
   B=@n70
   Alpha=PerInstanceFadeAmount()
  )
  n72 = WorldPosition()
  n73 = RuntimeVirtualTextureSampleParameter() {materialType=WorldHeight, parameterName=RVTHeight, virtualTexture=/Game/Athena/Apollo/Maps/Landscape/VT/RVT_Apollo_Landscape_Height}
  n74 = Subtract(A=Add(A=ScalarParameter() {group=RVT, parameterName=Offset RVT WP}, B=ComponentMask(None=@n72.XYZ) {b=true}), B=@n73.WorldHeight)
  n75 = BlendMaterialAttributes(
   A=@n67
   B=StaticSwitchParameter(
    True=SetMaterialAttributes!(MaterialAttributes=@n67, Normal=@n69.Normal) {attributeSetTypes [#4]}
    False=@n68
   ) {parameterName=RVT Colors}
   Alpha=Saturate(None=Divide(A=@n74, B=@n71))
  ) {vertexAttributeBlendType=UseA}
  n76 = NamedRerouteUsage()
  n77 = StaticSwitchParameter(
   True=MaterialFunctionCall(NoRVT=@n65, RVT=MaterialProxyReplace(Realtime=@n75, MaterialProxy=@n65)) {materialFunction=&FN_RVT_Switch_Material}.Result
   False=@n76
  ) {parameterName=UseRVT}
  ShadingPathSwitch(Default=@n77, Deferred=@n77, Forward=@n76, Mobile=@n76)
  n78 = ComponentMask(None=WorldPosition().XYZ) {b=true}
  n79 = Subtract(A=@n78, B=Constant() {r=-700})
  n80 = Subtract(A=@n78, B=Constant() {r=-1200})
  Subtract(A=OneMinus(None=Saturate(None=@n79)), B=OneMinus(None=Saturate(None=@n80)))
  TextureSample() {samplerType=SAMPLERTYPE_LinearColor, texture=/Game/Environments/Asteria/Foliage/ForestFloor/T_NITE_Foliage_LeafPile_A_D}
  VectorParameter() {parameterName=G_SnowColor, defaultValue=<rgba 0.6437 0.6862 0.75 1>}
  VertexInterpolator(VS=@n64)
 ]
}
